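# Quest entry row widget for the quest debug tool, showing a quest's definition and display data assets, display name, progress and state.
# SetData stores the CategoryQuestDebugData struct through a field-notify broadcast, and the event graph is empty.
# Helper functions supply display values: asset name or "[Not found]", achieved/required progress or "N/A", state enum as text, and the QuestView pip color.
# The layout adds a select button and browse buttons beside both asset fields.
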WidgetBlueprint /QuestUI/DebugTool/CategoryWidgets/QuestDebugToolCategoryQuest
graphs [
 EventGraph {links [], nodes []}
 GetAssetName {nodes [
   n1 = GetAssetName!() -> (then, Asset:Object Reference) {is=title K2Node_FunctionEntry}
   > n2 = Is Valid(InputObject=@n1)
   Return Node!(execute=@n2.Is Not Valid, AssetName:Text="LOCTEXT(\"[Not found]\")")
   > Return Node(AssetName="To Text (String)"(InString=GetObjectName(Object=@n1)))
  ]}
 GetPipColor {nodes [
   GetPipColor() {is=title K2Node_FunctionEntry}
   > Return Node!(
    execute
    Color:Linear Color Structure=Get Debug List Entry Type Color!(EntryType:EQuestPanelDebugListEntryType Enum=QuestView) -> (Color:Linear Color Structure) {is=title K2Node_CallFunction}
   )
  ]}
 GetProgress {nodes [
   n1 = GetProgress!() -> (then, Data:"Category Quest Debug Data Structure (by ref)") {is=title K2Node_FunctionEntry}
   n2 = Break Category Quest Debug Data!(CategoryQuestDebugData:"Category Quest Debug Data Structure (by ref)"=@n1) -> (
    QuestDefinition:Fort Quest Item Definition Object Reference
    QuestDisplayData:Fort Quest Display Data Asset Object Reference
    AchievedCount:int
    RequiredCount:int
   ) {is=title K2Node_BreakStruct}
   > n3 = Branch(Condition="integer > integer"(A=@n2.RequiredCount))
   > Return Node!(execute, Output:Text=Format Text!(Format="LOCTEXT(\"{0} / {1}\")", 0:int=@n2.AchievedCount, 1:int=@n2.RequiredCount) -> (Result))
   Return Node(execute=@n3.else, Output="LOCTEXT(\"N/A\")")
  ]}
 GetState {nodes [
   n1 = GetState!() -> (then, Data:"Category Quest Debug Data Structure (by ref)") {is=title K2Node_FunctionEntry}
   n2 = Break Category Quest Debug Data!(CategoryQuestDebugData=@n1) -> (QuestDefinition, QuestDisplayData, QuestState:EFortQuestState Enum)
   > Return Node(Output="To Text (String)"(InString=Enum to String!(Enumerator:EFortQuestState Enum=@n2.QuestState) -> (ReturnValue:String)))
  ]}
 SetData {nodes [
   n1 = SetData!() -> (then, Data:"Category Quest Debug Data Structure (by ref)") {is=title K2Node_FunctionEntry}
   > Set with Broadcast Data!(execute, Data:Category Quest Debug Data Structure=@n1) -> (then, Output_Get:Category Quest Debug Data Structure)
  ]}
]
variables [Data {
  type=CategoryQuestDebugData
  default {achievedCount=0, displayName="", listItemId="", questDefinition=null, questDisplayData=null, questState=Inactive, requiredCount=0}
 }]
widgets {
 info {inheritedWidgetCount=0, parentClass=&UserWidget, rootWidgetClass=&HorizontalBox, widgetCount=13}
 widgets [
  HorizontalBox_0 {widgetClassPath=&HorizontalBox, widget=HorizontalBox_0, properties {bOverride_Cursor=true}}
  Button_SelectEntry {
   widgetClassPath=/QuestUI/DebugTool/CommonWidgets/QuestDebugToolSelectEntryButton
   bIsVariable=true
   parent=HorizontalBox_0
   widget=Button_SelectEntry
   slot {class=HorizontalBoxSlot, properties {verticalAlignment=VAlign_Center, padding=<ltrb 4 4 4 4>}}
  }
  Pip {
   widgetClassPath=&Image
   bIsVariable=true
   parent=HorizontalBox_0
   widget=Pip
   properties {brush {
     bIsDynamicallyLoaded=false
     drawAs=Image
     imageType=NoImage
     mirroring=NoMirror
     resourceName=None
     resourceObject=null
     tiling=NoTile
     imageSize=<8 8>
     margin=<ltrb 0 0 0 0>
     outlineSettings {
      bUseBrushTransparency=false
      roundingType=HalfHeightRadius
      width=0
      color {colorUseRule=UseColor_Specified, specifiedColor=<rgba 0 0 0 0>}
      cornerRadii=<0 0 0 0>
     }
     tintColor {colorUseRule=UseColor_Specified, specifiedColor=<rgba 1 1 1 1>}
     uVRegion {bIsValid=false, max=<0 0>, min=<0 0>}
    }}
   slot {class=HorizontalBoxSlot, properties {padding=<ltrb 0 6 0 6>}}
  }
  GridPanel_0 {
   widgetClassPath=&GridPanel
   parent=HorizontalBox_0
   widget=GridPanel_0
   properties {bOverride_Cursor=true, columnFill [1,1,0.2]}
   slot {class=HorizontalBoxSlot, properties {size {sizeRule=Fill, value=1}}}
  }
  HorizontalBox {widgetClassPath=&HorizontalBox, parent=GridPanel_0, widget=HorizontalBox, slot {class=GridSlot}}
  Button_DefinitionBrowse {
   widgetClassPath=/QuestUI/DebugTool/CommonWidgets/QuestDebugToolBrowseButton
   bIsVariable=true
   parent=HorizontalBox
   widget=Button_DefinitionBrowse
   slot {class=HorizontalBoxSlot, properties {verticalAlignment=VAlign_Center, padding=<ltrb 4 4 4 4>}}
  }
  Widget_DefinitionField {
   widgetClassPath=/QuestUI/DebugTool/CommonWidgets/QuestDebugToolInfoField
   bIsVariable=true
   parent=HorizontalBox
   widget=Widget_DefinitionField
   properties {header=Quest Definition}
   slot {class=HorizontalBoxSlot, properties {size {sizeRule=Fill, value=1}}}
  }
  HorizontalBox_2 {widgetClassPath=&HorizontalBox, parent=GridPanel_0, widget=HorizontalBox_2, slot {class=GridSlot, properties {row=1}}}
  Button_DisplayDataBrowse {
   widgetClassPath=&QuestDebugToolBrowseButton
   bIsVariable=true
   parent=HorizontalBox_2
   widget=Button_DisplayDataBrowse
   slot {class=HorizontalBoxSlot, properties {verticalAlignment=VAlign_Center, padding=<ltrb 4 4 4 4>}}
  }
  Widget_DisplayDataField {
   widgetClassPath=&QuestDebugToolInfoField
   bIsVariable=true
   parent=HorizontalBox_2
   widget=Widget_DisplayDataField
   properties {header=Quest Display Data}
   slot {class=HorizontalBoxSlot, properties {size {sizeRule=Fill, value=1}}}
  }
  Widget_DisplayName {
   widgetClassPath=&QuestDebugToolInfoField
   bIsVariable=true
   parent=GridPanel_0
   widget=Widget_DisplayName
   properties {header=Display Name, multiline=true}
   slot {class=GridSlot, properties {column=1, rowSpan=2, padding=<ltrb 4 0 0 0>}}
  }
  Widget_Progress {
   widgetClassPath=&QuestDebugToolInfoField
   bIsVariable=true
   parent=GridPanel_0
   widget=Widget_Progress
   properties {header=Progress}
   slot {class=GridSlot, properties {column=2}}
  }
  Widget_State {
   widgetClassPath=&QuestDebugToolInfoField
   bIsVariable=true
   parent=GridPanel_0
   widget=Widget_State
   properties {header=State}
   slot {class=GridSlot, properties {column=2, row=1}}
  }
 ]
}
