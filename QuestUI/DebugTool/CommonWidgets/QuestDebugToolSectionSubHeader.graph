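# Reusable sub-header label widget for sections of the Quest debug tool UI.
# An Overlay stacks a padded, semi-transparent dark teal background image under a ScaleBox that holds a bold CommonTextBlock.
# PreConstruct copies the SubHeader variable into the text block, so the label also updates in designer preview.
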WidgetBlueprint /QuestUI/DebugTool/CommonWidgets/QuestDebugToolSectionSubHeader
graphs [EventGraph {nodes [
   Event PreConstruct()
   > "SetText (Text)"(self=Get Text_SubHeader() -> (Text_SubHeader:Common Text Object Reference), InText=Get SubHeader() -> (SubHeader:Text))
  ]}]
variables [SubHeader {type=string, default=SubHeader}]
widgets {
 info {inheritedWidgetCount=0, parentClass=&UserWidget, rootWidgetClass=&Overlay, widgetCount=4}
 widgets [
  Overlay_0 {widgetClassPath=&Overlay, widget=Overlay_0}
  BG {
   widgetClassPath=&Image
   parent=Overlay_0
   widget=BG
   properties {
    brush {
     bIsDynamicallyLoaded=false
     drawAs=Image
     imageType=NoImage
     mirroring=NoMirror
     resourceName=None
     resourceObject=null
     tiling=NoTile
     imageSize=<4 4>
     margin=<ltrb 0 0 0 0>
     outlineSettings {
      bUseBrushTransparency=false
      roundingType=HalfHeightRadius
      width=0
      color {colorUseRule=UseColor_Specified, specifiedColor=<rgba 0 0 0 0>}
      cornerRadii=<0 0 0 0>
     }
     tintColor {colorUseRule=UseColor_Specified, specifiedColor=<rgba 1 1 1 1>}
     uVRegion {bIsValid=false, max=<0 0>, min=<0 0>}
    }
    colorAndOpacity=<rgba 0 0.141 0.1615 0.5>
   }
   slot {class=OverlaySlot, properties {horizontalAlignment=HAlign_Fill, verticalAlignment=VAlign_Fill, padding=<ltrb 5 5 5 5>}}
  }
  ScaleBox_0 {
   widgetClassPath=&ScaleBox
   parent=Overlay_0
   widget=ScaleBox_0
   slot {class=OverlaySlot, properties {horizontalAlignment=HAlign_Fill, verticalAlignment=VAlign_Fill, padding=<ltrb 10 5 10 5>}}
  }
  Text_SubHeader {
   widgetClassPath=&CommonTextBlock
   bIsVariable=true
   parent=ScaleBox_0
   widget=Text_SubHeader
   properties {
    style=/FortUILibrary/Styles/Typography/Foundation/TextStyle-Base-XXS-B-Black.TextStyle-Base-XXS-B-Black_C
    text=SubHeader
    colorAndOpacity {colorUseRule=UseColor_Specified, specifiedColor=<rgba 0 0 0 1>}
    font {
     bForceMonospaced=false
     bMaterialIsStencil=false
     fontMaterial=null
     fontObject=&NotoSans
     letterSpacing=0
     monospacedWidth=1
     size=14
     skewAmount=0
     typefaceFontName=NotoSans-Bold
     outlineSettings {
      bApplyOutlineToDropShadows=false
      bMiteredCorners=false
      bSeparateFillAlpha=false
      outlineMaterial=null
      outlineSize=0
      outlineColor=<rgba 0 0 0 1>
     }
    }
    shadowOffset=<0 0>
   }
   slot {class=ScaleBoxSlot, properties {horizontalAlignment=HAlign_Fill, verticalAlignment=VAlign_Fill}}
  }
 ]
}
