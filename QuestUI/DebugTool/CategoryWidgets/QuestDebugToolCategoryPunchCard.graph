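# Quest debug tool widget that displays one punch card, with its bundle definition, display data, name, browse and select buttons, and a colored pip.
# SetData stores the data in a broadcasting field-notify variable, then rebuilds a Dynamic Entry Box with one category quest widget per quest.
# The expand widget drives a visibility switcher between the quest list and a collapsed placeholder.
# Helper functions supply asset names with a not-found fallback and the BundleView pip color.
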WidgetBlueprint /QuestUI/DebugTool/CategoryWidgets/QuestDebugToolCategoryPunchCard
graphs [
 CreateChildren {nodes [
   n1 = CreateChildren!() -> (then, Array:Array of Category Quest Debug Data Structures) {is=title K2Node_FunctionEntry}
   n2 = Get EntryBox_Quests() -> (EntryBox_Quests:Dynamic Entry Box Object Reference)
   > Reset!(execute, self:Dynamic Entry Box Object Reference=@n2, bDeleteWidgets:bool=false) -> (then)
   > n3 = For Each Loop(Array:Array of Category Quest Debug Data Structures=@n1) -> (Array Element:Category Quest Debug Data Structure)
   > n4 = Create Entry!(execute, self:Dynamic Entry Box Object Reference=@n2) -> (then, ReturnValue:User Widget Object Reference)
   > n5 = Cast To QuestDebugToolCategoryQuest!(execute, Object=@n4) -> (then, CastFailed, AsQuest Debug Tool Category Quest:Quest Debug Tool Category Quest Object Reference) {is=title K2Node_DynamicCast}
   > SetData!(execute, self:Quest Debug Tool Category Quest Object Reference=@n5, Data:"Category Quest Debug Data Structure (by ref)"=@n3) -> (then)
  ]}
 EventGraph {nodes [
   "On Expand Changed (Widget_Expand)"!() -> (OutputDelegate, then, IsExpanded:bool) {is=title K2Node_ComponentBoundEvent}
   > UpdateSwitcher()
  ]}
 GetAssetName {nodes [
   n1 = GetAssetName!() -> (then, Asset:Object Reference) {is=title K2Node_FunctionEntry}
   > n2 = Is Valid(InputObject=@n1)
   Return Node!(execute=@n2.Is Not Valid, AssetName:Text="LOCTEXT(\"[Not found]\")")
   > Return Node(AssetName="To Text (String)"(InString=GetObjectName(Object=@n1)))
  ]}
 GetPipColor {nodes [
   GetPipColor() {is=title K2Node_FunctionEntry}
   > Return Node!(
    execute
    Color:Linear Color Structure=Get Debug List Entry Type Color!(EntryType:EQuestPanelDebugListEntryType Enum=BundleView) -> (Color:Linear Color Structure) {is=title K2Node_CallFunction}
   )
  ]}
 SetData {nodes [
   n1 = SetData!() -> (then, Data:"Category Punch Card Debug Data Structure (by ref)") {is=title K2Node_FunctionEntry}
   > n2 = Set with Broadcast Data!(execute, Data:Category Punch Card Debug Data Structure=@n1) -> (then, Output_Get:Category Punch Card Debug Data Structure)
   > CreateChildren!(
    execute
    self
    Array:Array of Category Quest Debug Data Structures=Break Category Punch Card Debug Data!(CategoryPunchCardDebugData:"Category Punch Card Debug Data Structure (by ref)"=@n2) -> (DebugQuestData:Array of Category Quest Debug Data Structures) {is=title K2Node_BreakStruct}
   ) -> (then) {is=title K2Node_CallFunction}
  ]}
 UpdateSwitcher {nodes [
   UpdateSwitcher() {is=title K2Node_FunctionEntry}
   > n1 = Branch(Condition=IsExpanded!(self:Quest Debug Tool Expand Object Reference=Get Widget_Expand() -> (Widget_Expand:Quest Debug Tool Expand Object Reference)) -> (IsExpanded:bool))
   n2 = Get Switcher_Content() -> (Switcher_Content:Common Visibility Switcher Object Reference)
   > SetActiveWidget(self=@n2, Widget=Get EntryBox_Quests())
   SetActiveWidget(execute=@n1.else, self=@n2, Widget=Get Text_Collapsed() -> (Text_Collapsed:Common Text Object Reference))
  ]}
]
variables [Data {
  type=CategoryPunchCardDebugData
  default {
   bundleDefinition=null
   bundleDisplayData=null
   bundleListItemId=""
   debugQuestData []
   displayData {
    bAllowInspectItems=true
    bDisplayCommunityBadge=false
    bDisplayExpirationTime=true
    backgroundIcon=null
    headerText=""
    orientation=Vertical
    progressDisplayMode=DisplayCurrentStage
    sortOrder=0
    subHeaderText=""
   }
  }
 }]
widgets {
 info {inheritedWidgetCount=0, parentClass=&UserWidget, rootWidgetClass=&VerticalBox, widgetCount=18}
 widgets [
  VerticalBox_0 {widgetClassPath=&VerticalBox, widget=VerticalBox_0}
  Widget_Header {
   widgetClassPath=/QuestUI/DebugTool/CommonWidgets/QuestDebugToolSectionSubHeader
   bIsVariable=true
   parent=VerticalBox_0
   widget=Widget_Header
   properties {subHeader=Punch Card}
   slot {class=VerticalBoxSlot}
  }
  HorizontalBox_0 {
   widgetClassPath=&HorizontalBox
   parent=VerticalBox_0
   widget=HorizontalBox_0
   properties {bOverride_Cursor=true}
   slot {class=VerticalBoxSlot}
  }
  Button_SelectEntry {
   widgetClassPath=/QuestUI/DebugTool/CommonWidgets/QuestDebugToolSelectEntryButton
   bIsVariable=true
   parent=HorizontalBox_0
   widget=Button_SelectEntry
   slot {class=HorizontalBoxSlot, properties {verticalAlignment=VAlign_Center, padding=<ltrb 4 4 4 4>}}
  }
  Pip {
   widgetClassPath=&Image
   bIsVariable=true
   parent=HorizontalBox_0
   widget=Pip
   properties {brush {
     bIsDynamicallyLoaded=false
     drawAs=Image
     imageType=NoImage
     mirroring=NoMirror
     resourceName=None
     resourceObject=null
     tiling=NoTile
     imageSize=<8 8>
     margin=<ltrb 0 0 0 0>
     outlineSettings {
      bUseBrushTransparency=false
      roundingType=HalfHeightRadius
      width=0
      color {colorUseRule=UseColor_Specified, specifiedColor=<rgba 0 0 0 0>}
      cornerRadii=<0 0 0 0>
     }
     tintColor {colorUseRule=UseColor_Specified, specifiedColor=<rgba 1 1 1 1>}
     uVRegion {bIsValid=false, max=<0 0>, min=<0 0>}
    }}
   slot {class=HorizontalBoxSlot, properties {padding=<ltrb 0 6 0 6>}}
  }
  GridPanel_0 {
   widgetClassPath=&GridPanel
   parent=HorizontalBox_0
   widget=GridPanel_0
   properties {columnFill [1,1]}
   slot {class=HorizontalBoxSlot, properties {padding=<ltrb 4 0 0 0>, size {sizeRule=Fill, value=1}}}
  }
  Widget_DisplayName {
   widgetClassPath=/QuestUI/DebugTool/CommonWidgets/QuestDebugToolInfoField
   bIsVariable=true
   parent=GridPanel_0
   widget=Widget_DisplayName
   properties {header=Display Name, multiline=true}
   slot {class=GridSlot, properties {column=1, rowSpan=2}}
  }
  HorizontalBox {widgetClassPath=&HorizontalBox, parent=GridPanel_0, widget=HorizontalBox, slot {class=GridSlot}}
  Button_BrowseDefinition {
   widgetClassPath=/QuestUI/DebugTool/CommonWidgets/QuestDebugToolBrowseButton
   bIsVariable=true
   parent=HorizontalBox
   widget=Button_BrowseDefinition
   slot {class=HorizontalBoxSlot, properties {verticalAlignment=VAlign_Center, padding=<ltrb 4 4 4 4>}}
  }
  Widget_DefinitionField {
   widgetClassPath=&QuestDebugToolInfoField
   bIsVariable=true
   parent=HorizontalBox
   widget=Widget_DefinitionField
   properties {header=Bundle Definition}
   slot {class=HorizontalBoxSlot, properties {size {sizeRule=Fill, value=1}}}
  }
  HorizontalBox_2 {widgetClassPath=&HorizontalBox, parent=GridPanel_0, widget=HorizontalBox_2, slot {class=GridSlot, properties {row=1}}}
  Button_BrowseDisplayData {
   widgetClassPath=&QuestDebugToolBrowseButton
   bIsVariable=true
   parent=HorizontalBox_2
   widget=Button_BrowseDisplayData
   slot {class=HorizontalBoxSlot, properties {verticalAlignment=VAlign_Center, padding=<ltrb 4 4 4 4>}}
  }
  Widget_DisplayDataField {
   widgetClassPath=&QuestDebugToolInfoField
   bIsVariable=true
   parent=HorizontalBox_2
   widget=Widget_DisplayDataField
   properties {header=Bundle Display Data}
   slot {class=HorizontalBoxSlot, properties {size {sizeRule=Fill, value=1}}}
  }
  HorizontalBox_1 {widgetClassPath=&HorizontalBox, parent=VerticalBox_0, widget=HorizontalBox_1, slot {class=VerticalBoxSlot}}
  Widget_Expand {
   widgetClassPath=/QuestUI/DebugTool/CommonWidgets/QuestDebugToolExpand
   bIsVariable=true
   parent=HorizontalBox_1
   widget=Widget_Expand
   slot {class=HorizontalBoxSlot, properties {padding=<ltrb 0 0 5 0>}}
  }
  Switcher_Content {
   widgetClassPath=&CommonVisibilitySwitcher
   bIsVariable=true
   parent=HorizontalBox_1
   widget=Switcher_Content
   slot {class=HorizontalBoxSlot, properties {size {sizeRule=Fill, value=1}}}
  }
  EntryBox_Quests {
   widgetClassPath=&DynamicEntryBox
   bIsVariable=true
   parent=Switcher_Content
   widget=EntryBox_Quests
   properties {entryBoxType=Vertical, entryWidgetClass=&/QuestDebugToolCategoryQuest.QuestDebugToolCategoryQuest_C}
   slot {class=CommonVisibilitySwitcherSlot}
  }
  Text_Collapsed {
   widgetClassPath=&CommonTextBlock
   bIsVariable=true
   parent=Switcher_Content
   widget=Text_Collapsed
   properties {
    style=/QuestUI/DebugTool/TextStyles/TextStyle_DebugTool_Header.TextStyle_DebugTool_Header_C
    text="[Content collapsed]"
    font {
     bForceMonospaced=false
     bMaterialIsStencil=false
     fontMaterial=null
     fontObject=&NotoSans
     letterSpacing=0
     monospacedWidth=1
     size=9
     skewAmount=0
     typefaceFontName=NotoSans-Bold
     outlineSettings {
      bApplyOutlineToDropShadows=false
      bMiteredCorners=false
      bSeparateFillAlpha=false
      outlineMaterial=null
      outlineSize=0
      outlineColor=<rgba 0 0 0 1>
     }
    }
    shadowOffset=<0 0>
   }
   slot {class=CommonVisibilitySwitcherSlot, properties {verticalAlignment=VAlign_Center}}
  }
 ]
}
