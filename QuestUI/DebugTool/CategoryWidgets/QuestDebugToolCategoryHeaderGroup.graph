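# Quest debug tool widget that displays one category header group with its tag, sort order and generated status.
# SetData fills the info fields, flags non-generated groups as alerts and expands only generated ones, then calls CreateChildren.
# CreateChildren rebuilds one entry box with punch card widgets, created by explicit class, followed by default-class quest widgets.
# UpdateSwitcher shows the entries, collapsed text or no-content text, disabling the expander when empty.
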WidgetBlueprint /QuestUI/DebugTool/CategoryWidgets/QuestDebugToolCategoryHeaderGroup
graphs [
 CreateChildren {nodes [
   n1 = CreateChildren!() -> (then, CategoryHeaderGroupDebugData:"Category Header Group Debug Data Structure (by ref)") {is=title K2Node_FunctionEntry}
   > n2 = Sequence!(execute) -> (then_0, then_1, then_2:Exec, then_3:Exec)
   > Reset!(
    execute
    self:Dynamic Entry Box Object Reference=Get EntryBox_Content() -> (EntryBox_Content:Fort Dynamic Entry Box Object Reference)
    bDeleteWidgets:bool=false
   ) -> (then)
   n3 = Break Category Header Group Debug Data!(CategoryHeaderGroupDebugData:"Category Header Group Debug Data Structure (by ref)"=@n1) -> (
    bIsGenerated:bool=~
    HeaderGroupData:Quest Category Header Group Data Structure
    DebugPunchCardData:Array of Category Punch Card Debug Data Structures
    DebugQuestData:Array of Category Quest Debug Data Structures
   ) {is=title K2Node_BreakStruct}
   n4 = For Each Loop(Exec=@n2.then_1, Array:Array of Category Punch Card Debug Data Structures=@n3.DebugPunchCardData) -> (Array Element:Category Punch Card Debug Data Structure)
   n5 = Get EntryBox_Content()
   > n6 = Create Entry of Class!(
    execute
    self:Dynamic Entry Box Object Reference=@n5
    EntryClass:User Widget Class Reference=&/QuestDebugToolCategoryPunchCard.QuestDebugToolCategoryPunchCard_C
   ) -> (then, ReturnValue:Quest Debug Tool Category Punch Card Object Reference)
   > SetData!(execute, self:Quest Debug Tool Category Punch Card Object Reference=@n6, Data:"Category Punch Card Debug Data Structure (by ref)"=@n4) -> (then)
   n7 = For Each Loop(Exec=@n2.then_2, Array:Array of Category Quest Debug Data Structures=@n3.DebugQuestData) -> (Array Element:Category Quest Debug Data Structure)
   > n8 = Create Entry!(execute, self:Dynamic Entry Box Object Reference=@n5) -> (then, ReturnValue:User Widget Object Reference)
   > n9 = Cast To QuestDebugToolCategoryQuest!(execute, Object=@n8) -> (then, CastFailed, AsQuest Debug Tool Category Quest:Quest Debug Tool Category Quest Object Reference) {is=title K2Node_DynamicCast}
   > SetData(self:Quest Debug Tool Category Quest Object Reference=@n9, Data:"Category Quest Debug Data Structure (by ref)"=@n7)
   UpdateSwitcher(execute=@n2.then_3)
  ]}
 EventGraph {nodes [
   "On Expand Changed (Widget_Expand)"!() -> (OutputDelegate, then, IsExpanded:bool) {is=title K2Node_ComponentBoundEvent}
   > UpdateSwitcher()
  ]}
 SetData {nodes [
   n1 = SetData!() -> (then, HeaderGroupData:"Category Header Group Debug Data Structure (by ref)") {is=title K2Node_FunctionEntry}
   > n2 = Sequence()
   n3 = Break Category Header Group Debug Data(CategoryHeaderGroupDebugData=@n1) -> (bIsGenerated=~)
   n4 = Break Quest Category Header Group Data!(QuestCategoryHeaderGroupData:"Quest Category Header Group Data Structure (by ref)"=@n3.HeaderGroupData) -> (HeaderGroupTag:Gameplay Tag Structure, HeaderGroupQuestGiverIcon:Texture 2D Soft Object Reference, HeaderGroupSortOrder:int) {is=title K2Node_BreakStruct}
   > SetData(
    self:Quest Debug Tool Info Field Object Reference=Get Widget_Tag() -> (Widget_Tag:Quest Debug Tool Info Field Object Reference)
    Data:Text=GameplayTagToText!(Tag:"Gameplay Tag Structure (by ref)"=@n4, Fallback:Text="LOCTEXT(\"[Default Header Group]\")") -> (Output:Text)
   ) {is=title K2Node_CallFunction}
   SetData(
    execute=@n2.then_1
    self=Get Widget_SortOrder() -> (Widget_SortOrder:Quest Debug Tool Info Field Object Reference)
    Data="To Text (Integer)"(Value=@n4.HeaderGroupSortOrder)
   )
   n5 = Get Widget_Generate() -> (Widget_Generate:Quest Debug Tool Info Field Object Reference)
   SetData(execute=@n2.then_2, self=@n5, Data=SelectText(A="LOCTEXT(\"Yes\")", B="LOCTEXT(\"No\")", bPickA=@n3))
   > SetDisplayAsAlert!(execute, self:Quest Debug Tool Info Field Object Reference=@n5, InDisplayAsAlert:bool=NOT Boolean(A=@n3)) -> (then)
   > SetIsExpanded!(
    execute
    self:Quest Debug Tool Expand Object Reference=Get Widget_Expand() -> (Widget_Expand:Quest Debug Tool Expand Object Reference)
    IsExpanded:bool=@n3
   ) -> (then)
   CreateChildren!(execute=@n2.then_3, self, CategoryHeaderGroupDebugData:"Category Header Group Debug Data Structure (by ref)"=@n1) -> (then) {is=title K2Node_CallFunction}
  ]}
 TagToText {nodes [
   n1 = TagToText!() -> (then, Tag:"Gameplay Tag Structure (by ref)") {is=title K2Node_FunctionEntry}
   n2 = GetDebugStringFromGameplayTag!(GameplayTag:Gameplay Tag Structure=@n1) -> (ReturnValue:String)
   > Return Node!(
    execute
    Output:Text=Select(Option 0:Text="To Text (String)"(InString=@n2), Option 1:Text="LOCTEXT(\"(none)\")", Index=IsEmpty(InString=@n2)) -> (ReturnValue:Text)
   )
  ]}
 UpdateSwitcher {nodes [
   UpdateSwitcher() {is=title K2Node_FunctionEntry}
   n1 = Get EntryBox_Content()
   > n2 = Branch(Condition="Equal (Integer)"(A=GetNumEntries!(self:Dynamic Entry Box Base Object Reference=@n1) -> (ReturnValue:int=0)))
   n3 = Get Widget_Expand()
   > SetIsInteractable!(execute, self:Quest Debug Tool Expand Object Reference=@n3, IsInteractable:bool=false) -> (then)
   > SetActiveWidget(
    self=Get Switcher_Content() -> (Switcher_Content:Common Visibility Switcher Object Reference)
    Widget=Get Text_NoContent() -> (Text_NoContent:Common Text Object Reference)
   )
   SetIsInteractable(execute=@n2.else, self=@n3, IsInteractable=true)
   > n4 = Branch(Condition=IsExpanded!(self:Quest Debug Tool Expand Object Reference=Get Widget_Expand()) -> (IsExpanded:bool))
   n5 = Get Switcher_Content()
   > SetActiveWidget(self=@n5, Widget=Get EntryBox_Content())
   SetActiveWidget(execute=@n4.else, self=@n5, Widget=Get Text_Collapsed() -> (Text_Collapsed:Common Text Object Reference))
  ]}
]
variables []
widgets {
 info {inheritedWidgetCount=0, parentClass=&UserWidget, rootWidgetClass=&VerticalBox, widgetCount=12}
 widgets [
  VerticalBox_21 {widgetClassPath=&VerticalBox, widget=VerticalBox_21}
  Widget_Header {
   widgetClassPath=/QuestUI/DebugTool/CommonWidgets/QuestDebugToolSectionSubHeader
   bIsVariable=true
   parent=VerticalBox_21
   widget=Widget_Header
   properties {subHeader=Header Group}
   slot {class=VerticalBoxSlot}
  }
  HorizontalBox_0 {widgetClassPath=&HorizontalBox, parent=VerticalBox_21, widget=HorizontalBox_0, slot {class=VerticalBoxSlot}}
  Widget_Tag {
   widgetClassPath=/QuestUI/DebugTool/CommonWidgets/QuestDebugToolInfoField
   bIsVariable=true
   parent=HorizontalBox_0
   widget=Widget_Tag
   properties {header=Header Group Tag}
   slot {class=HorizontalBoxSlot, properties {size {sizeRule=Fill, value=1}}}
  }
  Widget_SortOrder {
   widgetClassPath=&QuestDebugToolInfoField
   bIsVariable=true
   parent=HorizontalBox_0
   widget=Widget_SortOrder
   properties {header=Header Group Sort Order}
   slot {class=HorizontalBoxSlot}
  }
  Widget_Generate {
   widgetClassPath=&QuestDebugToolInfoField
   bIsVariable=true
   parent=HorizontalBox_0
   widget=Widget_Generate
   properties {header=Is Generated}
   slot {class=HorizontalBoxSlot}
  }
  HorizontalBox_1 {widgetClassPath=&HorizontalBox, parent=VerticalBox_21, widget=HorizontalBox_1, slot {class=VerticalBoxSlot}}
  Widget_Expand {
   widgetClassPath=/QuestUI/DebugTool/CommonWidgets/QuestDebugToolExpand
   bIsVariable=true
   parent=HorizontalBox_1
   widget=Widget_Expand
   slot {class=HorizontalBoxSlot, properties {padding=<ltrb 0 0 5 0>}}
  }
  Switcher_Content {
   widgetClassPath=&CommonVisibilitySwitcher
   bIsVariable=true
   parent=HorizontalBox_1
   widget=Switcher_Content
   slot {class=HorizontalBoxSlot, properties {size {sizeRule=Fill, value=1}}}
  }
  EntryBox_Content {
   widgetClassPath=/Script/FortniteUI.FortDynamicEntryBox
   bIsVariable=true
   parent=Switcher_Content
   widget=EntryBox_Content
   properties {
    entryBoxType=Vertical
    entryWidgetClass=&/QuestDebugToolCategoryQuest.QuestDebugToolCategoryQuest_C
    numDesignerPreviewEntries=2
   }
   slot {class=CommonVisibilitySwitcherSlot}
  }
  Text_Collapsed {
   widgetClassPath=&CommonTextBlock
   bIsVariable=true
   parent=Switcher_Content
   widget=Text_Collapsed
   properties {
    style=/QuestUI/DebugTool/TextStyles/TextStyle_DebugTool_Header.TextStyle_DebugTool_Header_C
    text="[Content collapsed]"
    font {
     bForceMonospaced=false
     bMaterialIsStencil=false
     fontMaterial=null
     fontObject=&NotoSans
     letterSpacing=0
     monospacedWidth=1
     size=9
     skewAmount=0
     typefaceFontName=NotoSans-Bold
     outlineSettings {
      bApplyOutlineToDropShadows=false
      bMiteredCorners=false
      bSeparateFillAlpha=false
      outlineMaterial=null
      outlineSize=0
      outlineColor=<rgba 0 0 0 1>
     }
    }
    shadowOffset=<0 0>
   }
   slot {class=CommonVisibilitySwitcherSlot, properties {verticalAlignment=VAlign_Center}}
  }
  Text_NoContent {
   widgetClassPath=&CommonTextBlock
   bIsVariable=true
   parent=Switcher_Content
   widget=Text_NoContent
   properties {
    style=&TextStyle_DebugTool_Header_C
    text="[No content]"
    font {
     bForceMonospaced=false
     bMaterialIsStencil=false
     fontMaterial=null
     fontObject=&NotoSans
     letterSpacing=0
     monospacedWidth=1
     size=9
     skewAmount=0
     typefaceFontName=NotoSans-Bold
     outlineSettings {
      bApplyOutlineToDropShadows=false
      bMiteredCorners=false
      bSeparateFillAlpha=false
      outlineMaterial=null
      outlineSize=0
      outlineColor=<rgba 0 0 0 1>
     }
    }
    shadowOffset=<0 0>
   }
   slot {class=CommonVisibilitySwitcherSlot, properties {verticalAlignment=VAlign_Center}}
  }
 ]
}
