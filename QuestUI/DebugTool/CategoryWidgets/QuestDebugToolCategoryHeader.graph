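# Quest debug tool widget that displays one category header and its header groups.
# SetData fills the tag, name, description, sort order and generated fields, flags non-generated headers as an alert, and expands only generated ones.
# CreateChildren rebuilds the entry box with one QuestDebugToolCategoryHeaderGroup per group, then updates the switcher.
# The expand toggle switches between those groups and a collapsed-content label.
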WidgetBlueprint /QuestUI/DebugTool/CategoryWidgets/QuestDebugToolCategoryHeader
graphs [
 CreateChildren {nodes [
   n1 = CreateChildren!() -> (then, Array:Array of Category Header Group Debug Data Structures) {is=title K2Node_FunctionEntry}
   n2 = Get EntryBox_HeaderGroups() -> (EntryBox_HeaderGroups:Fort Dynamic Entry Box Object Reference)
   > Reset!(execute, self:Dynamic Entry Box Object Reference=@n2, bDeleteWidgets:bool=false) -> (then)
   > n3 = For Each Loop(Array:Array of Category Header Group Debug Data Structures=@n1) -> (Array Element:Category Header Group Debug Data Structure)
   UpdateSwitcher(execute=@n3.Completed)
   n4 = Create Entry!(execute=@n3.LoopBody, self:Dynamic Entry Box Object Reference=@n2) -> (then, ReturnValue:User Widget Object Reference)
   > n5 = Cast To QuestDebugToolCategoryHeaderGroup!(execute, Object=@n4) -> (then, CastFailed, AsQuest Debug Tool Category Header Group:Quest Debug Tool Category Header Group Object Reference) {is=title K2Node_DynamicCast}
   > SetData!(
    execute
    self:Quest Debug Tool Category Header Group Object Reference=@n5
    HeaderGroupData:"Category Header Group Debug Data Structure (by ref)"=@n3
   ) -> (then)
  ]}
 EventGraph {nodes [
   "On Expand Changed (Widget_Expand)"!() -> (OutputDelegate, then, IsExpanded:bool) {is=title K2Node_ComponentBoundEvent}
   > UpdateSwitcher()
  ]}
 SetData {nodes [
   n1 = SetData!() -> (then, HeaderData:"Category Header Debug Data Structure (by ref)") {is=title K2Node_FunctionEntry}
   > n2 = Sequence!(execute) -> (then_0, then_1, then_2:Exec, then_3:Exec, then_4:Exec, then_5:Exec)
   n3 = Break Category Header Debug Data!(CategoryHeaderDebugData:"Category Header Debug Data Structure (by ref)"=@n1) -> (
    bIsGenerated:bool=~
    HeaderData:Quest Category Header Data Structure
    DebugHeaderGroupData:Array of Category Header Group Debug Data Structures
   ) {is=title K2Node_BreakStruct}
   n4 = Break Quest Category Header Data!(QuestCategoryHeaderData:"Quest Category Header Data Structure (by ref)"=@n3.HeaderData) -> (
    HeaderName:Text
    HeaderDescription:Text
    bAlwaysShowHeaderDescription:bool=~
    HeaderTag:Gameplay Tag Structure
    HeaderSortOrder:int
    DefaultHeaderGroupSortOrder:int
   ) {is=title K2Node_BreakStruct}
   > SetData!(
    execute
    self:Quest Debug Tool Info Field Object Reference=Get Widget_Name() -> (Widget_Name:Quest Debug Tool Info Field Object Reference)
    Data:Text=TextOrFallback!(Input:Text=@n4, Fallback:Text="LOCTEXT(\"None\")") -> (Output:Text)
   ) -> (then) {is=title K2Node_CallFunction}
   SetData(
    execute=@n2.then_1
    self=Get Widget_Desc() -> (Widget_Desc:Quest Debug Tool Info Field Object Reference)
    Data=TextOrFallback(Input=@n4.HeaderDescription, Fallback="LOCTEXT(\"None\")")
   )
   SetData(
    execute=@n2.then_2
    self=Get Widget_Tag() -> (Widget_Tag:Quest Debug Tool Info Field Object Reference)
    Data=GameplayTagToText!(Tag:"Gameplay Tag Structure (by ref)"=@n4.HeaderTag, Fallback:Text="LOCTEXT(\"[Default Header]\")") -> (Output:Text)
   )
   n5 = Get Widget_Generate() -> (Widget_Generate:Quest Debug Tool Info Field Object Reference)
   SetData(execute=@n2.then_3, self=@n5, Data=SelectText(A="LOCTEXT(\"Yes\")", B="LOCTEXT(\"No\")", bPickA=@n3))
   > SetDisplayAsAlert!(execute, self:Quest Debug Tool Info Field Object Reference=@n5, InDisplayAsAlert:bool=NOT Boolean(A=@n3)) -> (then)
   > SetIsExpanded!(
    execute
    self:Quest Debug Tool Expand Object Reference=Get Widget_Expand() -> (Widget_Expand:Quest Debug Tool Expand Object Reference)
    IsExpanded:bool=@n3
   ) -> (then)
   SetData(
    execute=@n2.then_4
    self=Get Widget_SortOrder() -> (Widget_SortOrder:Quest Debug Tool Info Field Object Reference)
    Data="To Text (Integer)"(Value=@n4.HeaderSortOrder)
   )
   CreateChildren!(execute=@n2.then_5, self, Array:Array of Category Header Group Debug Data Structures=@n3.DebugHeaderGroupData) -> (then) {is=title K2Node_CallFunction}
  ]}
 UpdateSwitcher {nodes [
   UpdateSwitcher() {is=title K2Node_FunctionEntry}
   > n1 = Branch(Condition=IsExpanded!(self:Quest Debug Tool Expand Object Reference=Get Widget_Expand()) -> (IsExpanded:bool))
   n2 = Get Switcher_Content() -> (Switcher_Content:Common Visibility Switcher Object Reference)
   > SetActiveWidget(self=@n2, Widget=Get EntryBox_HeaderGroups())
   SetActiveWidget(execute=@n1.else, self=@n2, Widget=Get Text_Collapsed() -> (Text_Collapsed:Common Text Object Reference))
  ]}
]
variables []
widgets {
 info {inheritedWidgetCount=0, parentClass=&UserWidget, rootWidgetClass=&VerticalBox, widgetCount=13}
 widgets [
  VerticalBox_37 {widgetClassPath=&VerticalBox, widget=VerticalBox_37}
  Widget_Header {
   widgetClassPath=/QuestUI/DebugTool/CommonWidgets/QuestDebugToolSectionSubHeader
   bIsVariable=true
   parent=VerticalBox_37
   widget=Widget_Header
   properties {subHeader=Header}
   slot {class=VerticalBoxSlot}
  }
  HorizontalBox_0 {widgetClassPath=&HorizontalBox, parent=VerticalBox_37, widget=HorizontalBox_0, slot {class=VerticalBoxSlot}}
  Widget_Tag {
   widgetClassPath=/QuestUI/DebugTool/CommonWidgets/QuestDebugToolInfoField
   bIsVariable=true
   parent=HorizontalBox_0
   widget=Widget_Tag
   properties {header=Header Tag}
   slot {class=HorizontalBoxSlot, properties {size {sizeRule=Fill, value=1}}}
  }
  Widget_Name {
   widgetClassPath=&QuestDebugToolInfoField
   bIsVariable=true
   parent=HorizontalBox_0
   widget=Widget_Name
   properties {header=Header Name}
   slot {class=HorizontalBoxSlot, properties {size {sizeRule=Fill, value=1}}}
  }
  Widget_Desc {
   widgetClassPath=&QuestDebugToolInfoField
   bIsVariable=true
   parent=HorizontalBox_0
   widget=Widget_Desc
   properties {header=Header Description}
   slot {class=HorizontalBoxSlot, properties {size {sizeRule=Fill, value=1}}}
  }
  Widget_SortOrder {
   widgetClassPath=&QuestDebugToolInfoField
   bIsVariable=true
   parent=HorizontalBox_0
   widget=Widget_SortOrder
   properties {header=Header Sort Order}
   slot {class=HorizontalBoxSlot}
  }
  Widget_Generate {
   widgetClassPath=&QuestDebugToolInfoField
   bIsVariable=true
   parent=HorizontalBox_0
   widget=Widget_Generate
   properties {header=Is Generated}
   slot {class=HorizontalBoxSlot}
  }
  HorizontalBox_1 {widgetClassPath=&HorizontalBox, parent=VerticalBox_37, widget=HorizontalBox_1, slot {class=VerticalBoxSlot}}
  Widget_Expand {
   widgetClassPath=/QuestUI/DebugTool/CommonWidgets/QuestDebugToolExpand
   bIsVariable=true
   parent=HorizontalBox_1
   widget=Widget_Expand
   slot {class=HorizontalBoxSlot, properties {padding=<ltrb 0 0 5 0>}}
  }
  Switcher_Content {
   widgetClassPath=&CommonVisibilitySwitcher
   bIsVariable=true
   parent=HorizontalBox_1
   widget=Switcher_Content
   slot {class=HorizontalBoxSlot, properties {size {sizeRule=Fill, value=1}}}
  }
  EntryBox_HeaderGroups {
   widgetClassPath=/Script/FortniteUI.FortDynamicEntryBox
   bIsVariable=true
   parent=Switcher_Content
   widget=EntryBox_HeaderGroups
   properties {
    entryBoxType=Vertical
    entryWidgetClass=&/QuestDebugToolCategoryHeaderGroup.QuestDebugToolCategoryHeaderGroup_C
    entrySpacing=<0 5>
   }
   slot {class=CommonVisibilitySwitcherSlot}
  }
  Text_Collapsed {
   widgetClassPath=&CommonTextBlock
   bIsVariable=true
   parent=Switcher_Content
   widget=Text_Collapsed
   properties {
    style=/QuestUI/DebugTool/TextStyles/TextStyle_DebugTool_Header.TextStyle_DebugTool_Header_C
    text="[Content collapsed]"
    font {
     bForceMonospaced=false
     bMaterialIsStencil=false
     fontMaterial=null
     fontObject=&NotoSans
     letterSpacing=0
     monospacedWidth=1
     size=9
     skewAmount=0
     typefaceFontName=NotoSans-Bold
     outlineSettings {
      bApplyOutlineToDropShadows=false
      bMiteredCorners=false
      bSeparateFillAlpha=false
      outlineMaterial=null
      outlineSize=0
      outlineColor=<rgba 0 0 0 1>
     }
    }
    shadowOffset=<0 0>
   }
   slot {class=CommonVisibilitySwitcherSlot, properties {verticalAlignment=VAlign_Center}}
  }
 ]
}
