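# Quest debug tool field widget that displays an asset with a select-entry button, color pip, browse button and labeled info field.
# PreConstruct passes the configurable header to the info field.
# Helper functions resolve the asset's object name, or "[Not found]" when invalid, and dim opacity when no view exists.
# The pip color function picks the view or asset entry type by view presence and maps it to a debug list color.
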WidgetBlueprint /QuestUI/DebugTool/CommonWidgets/QuestDebugToolAssetField
graphs [
 EventGraph {nodes [
   Event PreConstruct()
   > SetHeader!(
    execute
    self:Quest Debug Tool Info Field Object Reference=Get Widget_AssetField() -> (Widget_AssetField:Quest Debug Tool Info Field Object Reference)
    Header:Text=Get AssetHeader() -> (AssetHeader:Text)
   ) -> (then)
  ]}
 GetAssetName {nodes [
   n1 = GetAssetName!() -> (then, Asset:Object Reference) {is=title K2Node_FunctionEntry}
   > n2 = Is Valid(InputObject=@n1)
   Return Node!(execute=@n2.Is Not Valid, AssetName:Text="LOCTEXT(\"[Not found]\")")
   > Return Node(AssetName="To Text (String)"(InString=GetObjectName(Object=@n1)))
  ]}
 GetContentOpacity {nodes [
   n1 = GetContentOpacity!() -> (then, bHasView:bool=~) {is=title K2Node_FunctionEntry}
   > Return Node!(execute, Opacity:double=Select(Option 0:double=0.5, Option 1:double=1, Index=@n1) -> (ReturnValue:double))
  ]}
 GetPipColor {nodes [
   n1 = GetPipColor!() -> (then, HasView:bool=~) {is=title K2Node_FunctionEntry}
   n2 = Get ViewEntryType() -> (ViewEntryType:EQuestPanelDebugListEntryType Enum=Unset)
   n3 = Get AssetEntryType() -> (AssetEntryType:EQuestPanelDebugListEntryType Enum=Unset)
   > Return Node!(
    execute
    Color:Linear Color Structure=Get Debug List Entry Type Color!(EntryType:EQuestPanelDebugListEntryType Enum=Select(Option 0:EQuestPanelDebugListEntryType Enum=@n3, Option 1:EQuestPanelDebugListEntryType Enum=@n2, Index=@n1) -> (ReturnValue:EQuestPanelDebugListEntryType Enum)) -> (Color:Linear Color Structure) {is=title K2Node_CallFunction}
   )
  ]}
]
variables [
 AssetEntryType {type=EQuestPanelDebugListEntryType, default=Unset}
 AssetHeader {type=string, default=Header}
 LocalAsset {type=Object, default=null}
 LocalAssetHasInstance {type=boolean}
 LocalListItemId {type=string, default=""}
 ViewEntryType {type=EQuestPanelDebugListEntryType, default=Unset}
]
widgets {
 info {inheritedWidgetCount=0, parentClass=/Script/QuestUI.QuestPanelDebugToolWidgetBase, rootWidgetClass=&HorizontalBox, widgetCount=6}
 widgets [
  HorizontalBox_0 {widgetClassPath=&HorizontalBox, widget=HorizontalBox_0}
  Button_SelectEntry {
   widgetClassPath=&/QuestDebugToolSelectEntryButton
   bIsVariable=true
   parent=HorizontalBox_0
   widget=Button_SelectEntry
   slot {class=HorizontalBoxSlot, properties {verticalAlignment=VAlign_Center, padding=<ltrb 4 4 4 4>}}
  }
  Pip {
   widgetClassPath=&Image
   bIsVariable=true
   parent=HorizontalBox_0
   widget=Pip
   properties {brush {
     bIsDynamicallyLoaded=false
     drawAs=Image
     imageType=NoImage
     mirroring=NoMirror
     resourceName=None
     resourceObject=null
     tiling=NoTile
     imageSize=<8 8>
     margin=<ltrb 0 0 0 0>
     outlineSettings {
      bUseBrushTransparency=false
      roundingType=HalfHeightRadius
      width=0
      color {colorUseRule=UseColor_Specified, specifiedColor=<rgba 0 0 0 0>}
      cornerRadii=<0 0 0 0>
     }
     tintColor {colorUseRule=UseColor_Specified, specifiedColor=<rgba 1 1 1 1>}
     uVRegion {bIsValid=false, max=<0 0>, min=<0 0>}
    }}
   slot {class=HorizontalBoxSlot, properties {horizontalAlignment=HAlign_Left}}
  }
  HorizontalBox_62 {
   widgetClassPath=&HorizontalBox
   parent=HorizontalBox_0
   widget=HorizontalBox_62
   slot {class=HorizontalBoxSlot, properties {size {sizeRule=Fill, value=1}}}
  }
  Button_AssetBrowse {
   widgetClassPath=&/QuestDebugToolBrowseButton
   bIsVariable=true
   parent=HorizontalBox_62
   widget=Button_AssetBrowse
   slot {class=HorizontalBoxSlot, properties {verticalAlignment=VAlign_Center, padding=<ltrb 4 4 4 4>}}
  }
  Widget_AssetField {
   widgetClassPath=&/QuestDebugToolInfoField
   bIsVariable=true
   parent=HorizontalBox_62
   widget=Widget_AssetField
   slot {class=HorizontalBoxSlot, properties {size {sizeRule=Fill, value=1}}}
  }
 ]
}
