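# Quest debug tool widget that lays out a "Category Info" section for a quest category.
# Under the section header it stacks a category asset field, a display name field and a multiline include tags field that fills the remaining height.
# GetIncludeTags turns a category data object's IncludeTags container into debug-string text, guarded by a validity check.
# The event graph is empty, and no wiring from that function to the fields appears here.
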WidgetBlueprint /QuestUI/DebugTool/CategoryWidgets/QuestDebugToolCategoryInfo
graphs [
 EventGraph {links [], nodes []}
 GetIncludeTags {nodes [
   n1 = GetIncludeTags!() -> (then, CategoryData:Quest Category Data Object Reference) {is=title K2Node_FunctionEntry}
   > n2 = Is Valid(InputObject=@n1)
   Return Node!(execute=@n2.Is Not Valid, TagText:Text)
   n3 = Get IncludeTags!(self:Quest Category Data Object Reference=@n1) -> (IncludeTags:Gameplay Tag Container Structure)
   > Return Node(TagText="To Text (String)"(InString=GetDebugStringFromGameplayTagContainer!(TagContainer:"Gameplay Tag Container Structure (by ref)"=@n3) -> (ReturnValue:String)))
  ]}
]
variables []
widgets {
 info {inheritedWidgetCount=0, parentClass=&UserWidget, rootWidgetClass=&VerticalBox, widgetCount=6}
 widgets [
  VerticalBox_1 {widgetClassPath=&VerticalBox, widget=VerticalBox_1}
  SectionHeader {
   widgetClassPath=/QuestUI/DebugTool/CommonWidgets/QuestDebugToolSectionHeader
   bIsVariable=true
   parent=VerticalBox_1
   widget=SectionHeader
   properties {header=Category Info}
   slot {class=VerticalBoxSlot}
  }
  VerticalBox_0 {widgetClassPath=&VerticalBox, parent=VerticalBox_1, widget=VerticalBox_0, slot {class=VerticalBoxSlot}}
  Widget_CategoryField {
   widgetClassPath=/QuestUI/DebugTool/CommonWidgets/QuestDebugToolAssetField
   bIsVariable=true
   parent=VerticalBox_0
   widget=Widget_CategoryField
   properties {assetEntryType=CategoryAsset, assetHeader=Category Data, viewEntryType=CategoryInstance}
   slot {class=VerticalBoxSlot}
  }
  Widget_DisplayName {
   widgetClassPath=/QuestUI/DebugTool/CommonWidgets/QuestDebugToolInfoField
   bIsVariable=true
   parent=VerticalBox_0
   widget=Widget_DisplayName
   properties {header=Display Name}
   slot {class=VerticalBoxSlot}
  }
  Widget_IncludeTags {
   widgetClassPath=&QuestDebugToolInfoField
   bIsVariable=true
   parent=VerticalBox_0
   widget=Widget_IncludeTags
   properties {header=Include Tags, multiline=true}
   slot {class=VerticalBoxSlot, properties {size {sizeRule=Fill, value=1}}}
  }
 ]
}
