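# A quest debug tool widget that shows category organization debug data under a "Category Organization" section header.
# SetData receives a category organization debug struct and resets a vertical Dynamic Entry Box.
# It then creates a compatibility group child widget for each entry in the struct's compatibility data array and passes that entry to the child's own SetData.
# The event graph is empty, so the display updates only when SetData is called.
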WidgetBlueprint /QuestUI/DebugTool/CategoryWidgets/QuestDebugToolCategoryOrganization
graphs [
 EventGraph {links [], nodes []}
 SetData {nodes [
   n1 = SetData!() -> (then, OrganizationData:"Category Organization Debug Data Structure (by ref)") {is=title K2Node_FunctionEntry}
   n2 = Get EntryBox_CompatibilityGroups() -> (EntryBox_CompatibilityGroups:Dynamic Entry Box Object Reference)
   > Reset!(execute, self:Dynamic Entry Box Object Reference=@n2, bDeleteWidgets:bool=false) -> (then)
   > n3 = For Each Loop(Array:Array of Category Compatibility Debug Data Structures=Break Category Organization Debug Data!(CategoryOrganizationDebugData:"Category Organization Debug Data Structure (by ref)"=@n1) -> (DebugCompatibilityData:Array of Category Compatibility Debug Data Structures) {is=title K2Node_BreakStruct}) -> (Array Element:Category Compatibility Debug Data Structure)
   > n4 = Create Entry!(execute, self:Dynamic Entry Box Object Reference=@n2) -> (then, ReturnValue:User Widget Object Reference)
   > n5 = Cast To QuestDebugToolCategoryCompatibilityGroup!(execute, Object=@n4) -> (then, CastFailed, AsQuest Debug Tool Category Compatibility Group:Quest Debug Tool Category Compatibility Group Object Reference) {is=title K2Node_DynamicCast}
   > SetData!(
    execute
    self:Quest Debug Tool Category Compatibility Group Object Reference=@n5
    CompatibilityData:"Category Compatibility Debug Data Structure (by ref)"=@n3
   ) -> (then) {is=title K2Node_CallFunction}
  ]}
]
variables []
widgets {
 info {inheritedWidgetCount=0, parentClass=&UserWidget, rootWidgetClass=&VerticalBox, widgetCount=3}
 widgets [
  VerticalBox_1 {widgetClassPath=&VerticalBox, widget=VerticalBox_1}
  SectionHeader {
   widgetClassPath=/QuestUI/DebugTool/CommonWidgets/QuestDebugToolSectionHeader
   bIsVariable=true
   parent=VerticalBox_1
   widget=SectionHeader
   properties {header=Category Organization}
   slot {class=VerticalBoxSlot}
  }
  EntryBox_CompatibilityGroups {
   widgetClassPath=&DynamicEntryBox
   bIsVariable=true
   parent=VerticalBox_1
   widget=EntryBox_CompatibilityGroups
   properties {
    entryBoxType=Vertical
    entryWidgetClass=&/QuestDebugToolCategoryCompatibilityGroup.QuestDebugToolCategoryCompatibilityGroup_C
    numDesignerPreviewEntries=1
    entrySpacing=<0 5>
   }
   slot {class=VerticalBoxSlot, properties {padding=<ltrb 10 0 10 0>}}
  }
 ]
}
