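# Quest debug tool widget that shows one Quest Product data asset as an icon beside its name and tags.
# SetData fills the name text from DisplayName, converts ContentRootTags to a debug string for the tags text, and passes the soft Icon reference to a lazy image, matching texture size.
# On Construct, the widget collapses itself unless InEditorOrPIE returns true.
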WidgetBlueprint /QuestUI/DebugTool/CategoryWidgets/QuestDebugToolProduct
graphs [
 EventGraph {nodes [
   Event Construct()
   > SetVisibility(InVisibility=Select(Option 0=Collapsed, Option 1=Visible, Index=InEditorOrPIE!() -> (ReturnValue:bool)))
  ]}
 SetData {nodes [
   n1 = SetData!() -> (then, Data:"Quest Product Data Asset Object Reference (by ref)") {is=title K2Node_FunctionEntry}
   > "SetText (Text)"(
    self=Get Text_DisplayName() -> (Text_DisplayName:Common Text Object Reference)
    InText=Get DisplayName!(self:Quest Product Data Asset Object Reference=@n1) -> (DisplayName:Text)
   )
   n2 = Get ContentRootTags!(self:Quest Product Data Asset Object Reference=@n1) -> (ContentRootTags:Gameplay Tag Container Structure)
   > "SetText (Text)"(
    self=Get Text_ProductTags() -> (Text_ProductTags:Common Text Object Reference)
    InText="To Text (String)"(InString=GetDebugStringFromGameplayTagContainer!(TagContainer:"Gameplay Tag Container Structure (by ref)"=@n2) -> (ReturnValue:String))
   )
   > SetBrushFromLazyDisplayAsset!(
    execute
    self:Common Lazy Image Object Reference=Get LazyImage_ProductIcon() -> (LazyImage_ProductIcon:Fort Lazy Image Object Reference)
    LazyObject:"Soft Object Reference (by ref)"=Get Icon!(self:Quest Product Data Asset Object Reference=@n1) -> (Icon:Material Instance Soft Object Reference)
    bMatchTextureSize:bool=true
   ) -> (then)
  ]}
]
variables []
widgets {
 info {inheritedWidgetCount=0, parentClass=/Script/QuestUI.QuestPanelDebugToolWidgetBase, rootWidgetClass=&Overlay, widgetCount=6}
 widgets [
  Overlay_0 {widgetClassPath=&Overlay, widget=Overlay_0}
  HorizontalBox_2 {
   widgetClassPath=&HorizontalBox
   parent=Overlay_0
   widget=HorizontalBox_2
   slot {class=OverlaySlot, properties {horizontalAlignment=HAlign_Fill, verticalAlignment=VAlign_Fill}}
  }
  LazyImage_ProductIcon {
   widgetClassPath=&FortLazyImage
   bIsVariable=true
   parent=HorizontalBox_2
   widget=LazyImage_ProductIcon
   slot {class=HorizontalBoxSlot, properties {horizontalAlignment=HAlign_Center, verticalAlignment=VAlign_Center}}
  }
  VerticalBox_0 {
   widgetClassPath=&VerticalBox
   parent=HorizontalBox_2
   widget=VerticalBox_0
   slot {class=HorizontalBoxSlot, properties {verticalAlignment=VAlign_Center, padding=<ltrb 5 0 0 0>, size {sizeRule=Fill, value=1}}}
  }
  Text_DisplayName {
   widgetClassPath=&CommonTextBlock
   bIsVariable=true
   parent=VerticalBox_0
   widget=Text_DisplayName
   properties {
    style=/QuestUI/DebugTool/TextStyles/TextStyle_DebugTool_CategoryHeader.TextStyle_DebugTool_CategoryHeader_C
    text=Product Name
    colorAndOpacity {colorUseRule=UseColor_Specified, specifiedColor=<rgba 1 0.9301 0 1>}
    font {
     bForceMonospaced=false
     bMaterialIsStencil=false
     fontMaterial=null
     fontObject=&NotoSans
     letterSpacing=0
     monospacedWidth=1
     size=10.5
     skewAmount=0
     typefaceFontName=NotoSans-Bold
     outlineSettings {
      bApplyOutlineToDropShadows=false
      bMiteredCorners=false
      bSeparateFillAlpha=false
      outlineMaterial=null
      outlineSize=0
      outlineColor=<rgba 0 0 0 1>
     }
    }
    shadowOffset=<0 0>
   }
   slot {class=VerticalBoxSlot}
  }
  Text_ProductTags {
   widgetClassPath=&CommonTextBlock
   bIsVariable=true
   parent=VerticalBox_0
   widget=Text_ProductTags
   properties {
    style=/QuestUI/DebugTool/TextStyles/TextStyle_DebugTool_CategoryBody.TextStyle_DebugTool_CategoryBody_C
    text=Product Tags
    colorAndOpacity {colorUseRule=UseColor_Specified, specifiedColor=<rgba 1 1 1 0.7>}
    font {
     bForceMonospaced=false
     bMaterialIsStencil=false
     fontMaterial=null
     fontObject=&NotoSans
     letterSpacing=0
     monospacedWidth=1
     size=9
     skewAmount=0
     typefaceFontName=NotoSans-Bold
     outlineSettings {
      bApplyOutlineToDropShadows=false
      bMiteredCorners=false
      bSeparateFillAlpha=false
      outlineMaterial=null
      outlineSize=0
      outlineColor=<rgba 0 0 0 1>
     }
    }
    shadowOffset=<0 0>
   }
   slot {class=VerticalBoxSlot}
  }
 ]
}
