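# Quest debug tool widget that shows the Products section of a quest panel debug category.
# SetData takes a category data struct, shows a red warning only when active product injection is in use, and resets the products entry box.
# For a non-empty product list it creates a QuestDebugToolProduct entry per valid data asset and passes it that asset.
# Otherwise, a visibility switcher displays a red no-matching-products message.
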WidgetBlueprint /QuestUI/DebugTool/CategoryWidgets/QuestDebugToolCategoryProducts
graphs [
 EventGraph {links [], nodes []}
 SetData {nodes [
   n1 = SetData!() -> (then, Data:"Quest Panel Debug Category Data Structure (by ref)") {is=title K2Node_FunctionEntry}
   n2 = Break Quest Panel Debug Category Data!(QuestPanelDebugCategoryData:"Quest Panel Debug Category Data Structure (by ref)"=@n1) -> (bActiveProductInjectionUsed:bool=~, ProductList:Array of Quest Product Data Asset Object References) {is=title K2Node_BreakStruct}
   > SetVisibility(
    self=Get Text_ActiveInjection() -> (Text_ActiveInjection:Common Text Object Reference)
    InVisibility=Select(Option 0=Collapsed, Option 1=Visible, Index=@n2)
   )
   > Reset!(
    execute
    self:Dynamic Entry Box Object Reference=Get EntryBox_Products() -> (EntryBox_Products:Dynamic Entry Box Object Reference)
    bDeleteWidgets:bool=false
   ) -> (then)
   > n3 = Branch(Condition="integer > integer"(A=Length(TargetArray:Array of Quest Product Data Asset Object References=@n2.ProductList)))
   n4 = Get Switcher_Products() -> (Switcher_Products:Common Visibility Switcher Object Reference)
   > SetActiveWidget(self=@n4, Widget=Get EntryBox_Products())
   > n5 = For Each Loop(Array:Array of Quest Product Data Asset Object References=@n2.ProductList) -> (Array Element:Quest Product Data Asset Object Reference)
   > Is Valid(InputObject=@n5)
   > n6 = Create Entry!(execute, self:Dynamic Entry Box Object Reference=Get EntryBox_Products()) -> (then, ReturnValue:User Widget Object Reference)
   > n7 = Cast To QuestDebugToolProduct!(execute, Object=@n6) -> (then, CastFailed, AsQuest Debug Tool Product:Quest Debug Tool Product Object Reference) {is=title K2Node_DynamicCast}
   > SetData!(execute, self:Quest Debug Tool Product Object Reference=@n7, Data:"Quest Product Data Asset Object Reference (by ref)"=@n5) -> (then) {is=title K2Node_CallFunction}
   SetActiveWidget(execute=@n3.else, self=@n4, Widget=Get Text_NoProducts() -> (Text_NoProducts:Common Text Object Reference))
  ]}
]
variables []
widgets {
 info {inheritedWidgetCount=0, parentClass=&UserWidget, rootWidgetClass=&Overlay, widgetCount=9}
 widgets [
  Overlay_0 {widgetClassPath=&Overlay, widget=Overlay_0}
  BG {
   widgetClassPath=&Image
   parent=Overlay_0
   widget=BG
   properties {colorAndOpacity=<rgba 0 0.3184 0.3646 0.1>}
   slot {class=OverlaySlot, properties {horizontalAlignment=HAlign_Fill, verticalAlignment=VAlign_Fill, padding=<ltrb 3 3 3 3>}}
  }
  VerticalBox_0 {
   widgetClassPath=&VerticalBox
   parent=Overlay_0
   widget=VerticalBox_0
   slot {class=OverlaySlot, properties {padding=<ltrb 5 5 5 5>}}
  }
  ScaleBox_0 {widgetClassPath=&ScaleBox, parent=VerticalBox_0, widget=ScaleBox_0, slot {class=VerticalBoxSlot}}
  Text_Header {
   widgetClassPath=&CommonTextBlock
   bIsVariable=true
   parent=ScaleBox_0
   widget=Text_Header
   properties {
    style=/QuestUI/DebugTool/TextStyles/TextStyle_DebugTool_Header.TextStyle_DebugTool_Header_C
    text=Products
    font {
     bForceMonospaced=false
     bMaterialIsStencil=false
     fontMaterial=null
     fontObject=&NotoSans
     letterSpacing=0
     monospacedWidth=1
     size=9
     skewAmount=0
     typefaceFontName=NotoSans-Bold
     outlineSettings {
      bApplyOutlineToDropShadows=false
      bMiteredCorners=false
      bSeparateFillAlpha=false
      outlineMaterial=null
      outlineSize=0
      outlineColor=<rgba 0 0 0 1>
     }
    }
    shadowOffset=<0 0>
   }
   slot {class=ScaleBoxSlot, properties {horizontalAlignment=HAlign_Fill, verticalAlignment=VAlign_Fill}}
  }
  Text_ActiveInjection {
   widgetClassPath=&CommonTextBlock
   bIsVariable=true
   parent=VerticalBox_0
   widget=Text_ActiveInjection
   properties {
    text="Active Product Injection In Use!"
    colorAndOpacity {colorUseRule=UseColor_Specified, specifiedColor=<rgba 1 0 0 1>}
    font {
     bForceMonospaced=false
     bMaterialIsStencil=false
     fontMaterial=null
     fontObject=&NotoSans
     letterSpacing=0
     monospacedWidth=1
     size=9
     skewAmount=0
     typefaceFontName=NotoSans-Bold
     outlineSettings {
      bApplyOutlineToDropShadows=false
      bMiteredCorners=false
      bSeparateFillAlpha=false
      outlineMaterial=null
      outlineSize=0
      outlineColor=<rgba 0 0 0 1>
     }
    }
    shadowOffset=<0 0>
   }
   slot {class=VerticalBoxSlot, properties {padding=<ltrb 0 4 0 0>}}
  }
  Switcher_Products {
   widgetClassPath=&CommonVisibilitySwitcher
   bIsVariable=true
   parent=VerticalBox_0
   widget=Switcher_Products
   slot {class=VerticalBoxSlot, properties {padding=<ltrb 0 4 0 0>}}
  }
  EntryBox_Products {
   widgetClassPath=&DynamicEntryBox
   bIsVariable=true
   parent=Switcher_Products
   widget=EntryBox_Products
   properties {entryBoxType=Vertical, entryWidgetClass=&/QuestDebugToolProduct.QuestDebugToolProduct_C}
   slot {class=CommonVisibilitySwitcherSlot}
  }
  Text_NoProducts {
   widgetClassPath=&CommonTextBlock
   bIsVariable=true
   parent=Switcher_Products
   widget=Text_NoProducts
   properties {
    text=(No Matching Products)
    colorAndOpacity {colorUseRule=UseColor_Specified, specifiedColor=<rgba 1 0 0 1>}
    font {
     bForceMonospaced=false
     bMaterialIsStencil=false
     fontMaterial=null
     fontObject=&NotoSans
     letterSpacing=0
     monospacedWidth=1
     size=9
     skewAmount=0
     typefaceFontName=NotoSans-Bold
     outlineSettings {
      bApplyOutlineToDropShadows=false
      bMiteredCorners=false
      bSeparateFillAlpha=false
      outlineMaterial=null
      outlineSize=0
      outlineColor=<rgba 0 0 0 1>
     }
    }
    shadowOffset=<0 0>
   }
   slot {class=CommonVisibilitySwitcherSlot, properties {padding=<ltrb 0 4 0 0>}}
  }
 ]
}
