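# QuestDebugToolProductInfo is a UserWidget that forms the Product Info category panel of the quest debug tool UI.
# A vertical box stacks a shared QuestDebugToolSectionHeader titled "Product Info" above a nested box holding a QuestDebugToolCategoryProducts widget set to fill.
# It is layout only, with an empty event graph and no variables of its own, so any product display logic sits outside this widget.
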WidgetBlueprint /QuestUI/DebugTool/CategoryWidgets/QuestDebugToolProductInfo
graphs [EventGraph {links [], nodes []}]
variables []
widgets {
 info {inheritedWidgetCount=0, parentClass=&UserWidget, rootWidgetClass=&VerticalBox, widgetCount=4}
 widgets [
  VerticalBox_1 {widgetClassPath=&VerticalBox, widget=VerticalBox_1}
  SectionHeader {
   widgetClassPath=/QuestUI/DebugTool/CommonWidgets/QuestDebugToolSectionHeader
   bIsVariable=true
   parent=VerticalBox_1
   widget=SectionHeader
   properties {header=Product Info}
   slot {class=VerticalBoxSlot}
  }
  VerticalBox_0 {widgetClassPath=&VerticalBox, parent=VerticalBox_1, widget=VerticalBox_0, slot {class=VerticalBoxSlot}}
  Widget_Products {
   widgetClassPath=&/QuestDebugToolCategoryProducts
   bIsVariable=true
   parent=VerticalBox_0
   widget=Widget_Products
   slot {class=VerticalBoxSlot, properties {size {sizeRule=Fill, value=1}}}
  }
 ]
}
