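# Quest debug tool row widget for one category entry, with select and browse buttons, a pip, and Category Data and Display Name fields.
# SetData stores the category struct in Entry with a field-notify broadcast.
# Helper functions turn an asset into a name, or "[Not found]", and pick a pip color by whether the category has an instance.
# On Construct, the widget collapses itself outside editor or PIE.
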WidgetBlueprint /QuestUI/DebugTool/CategoryWidgets/QuestDebugToolCategoryEntry
graphs [
 EventGraph {nodes [
   Event Construct()
   > SetVisibility(InVisibility=Select(Option 0=Collapsed, Option 1=Visible, Index=InEditorOrPIE!() -> (ReturnValue:bool)))
  ]}
 GetAssetName {nodes [
   n1 = GetAssetName!() -> (then, Asset:Object Reference) {is=title K2Node_FunctionEntry}
   > n2 = Is Valid(InputObject=@n1)
   Return Node!(execute=@n2.Is Not Valid, AssetName:Text="LOCTEXT(\"[Not found]\")")
   > Return Node(AssetName="To Text (String)"(InString=GetObjectName(Object=@n1)))
  ]}
 GetPipColor {nodes [
   n1 = GetPipColor!() -> (then, Entry:"Quest Panel Debug Category Data Structure (by ref)") {is=title K2Node_FunctionEntry}
   n2 = Break Quest Panel Debug Category Data!(QuestPanelDebugCategoryData:"Quest Panel Debug Category Data Structure (by ref)"=@n1) -> (bCategoryHasInstance:bool=~) {is=title K2Node_BreakStruct}
   > Return Node!(
    execute
    Color:Linear Color Structure=Get Debug List Entry Type Color!(EntryType:EQuestPanelDebugListEntryType Enum=Select(Option 0:EQuestPanelDebugListEntryType Enum=CategoryAsset, Option 1:EQuestPanelDebugListEntryType Enum=CategoryInstance, Index=@n2) -> (ReturnValue:EQuestPanelDebugListEntryType Enum)) -> (Color:Linear Color Structure) {is=title K2Node_CallFunction}
   )
  ]}
 SetData {nodes [
   n1 = SetData!() -> (then, Data:"Quest Panel Debug Category Data Structure (by ref)") {is=title K2Node_FunctionEntry}
   > Set with Broadcast Entry!(execute, Entry:Quest Panel Debug Category Data Structure=@n1) -> (then, Output_Get:Quest Panel Debug Category Data Structure)
  ]}
]
variables [Entry {
  type=QuestPanelDebugCategoryData
  default {bActiveProductInjectionUsed=false, bCategoryHasInstance=false, categoryData=null, categoryListItemId="", productList []}
 }]
widgets {
 info {inheritedWidgetCount=0, parentClass=/Script/QuestUI.QuestPanelDebugToolWidgetBase, rootWidgetClass=&HorizontalBox, widgetCount=6}
 widgets [
  HorizontalBox_2 {widgetClassPath=&HorizontalBox, widget=HorizontalBox_2}
  Button_SelectEntry {
   widgetClassPath=/QuestUI/DebugTool/CommonWidgets/QuestDebugToolSelectEntryButton
   bIsVariable=true
   parent=HorizontalBox_2
   widget=Button_SelectEntry
   slot {class=HorizontalBoxSlot, properties {verticalAlignment=VAlign_Center, padding=<ltrb 4 4 4 4>}}
  }
  Pip {
   widgetClassPath=&Image
   bIsVariable=true
   parent=HorizontalBox_2
   widget=Pip
   properties {brush {
     bIsDynamicallyLoaded=false
     drawAs=Image
     imageType=NoImage
     mirroring=NoMirror
     resourceName=None
     resourceObject=null
     tiling=NoTile
     imageSize=<8 8>
     margin=<ltrb 0 0 0 0>
     outlineSettings {
      bUseBrushTransparency=false
      roundingType=HalfHeightRadius
      width=0
      color {colorUseRule=UseColor_Specified, specifiedColor=<rgba 0 0 0 0>}
      cornerRadii=<0 0 0 0>
     }
     tintColor {colorUseRule=UseColor_Specified, specifiedColor=<rgba 1 1 1 1>}
     uVRegion {bIsValid=false, max=<0 0>, min=<0 0>}
    }}
   slot {class=HorizontalBoxSlot, properties {padding=<ltrb 0 6 0 6>}}
  }
  Button_BrowseToAsset {
   widgetClassPath=/QuestUI/DebugTool/CommonWidgets/QuestDebugToolBrowseButton
   bIsVariable=true
   parent=HorizontalBox_2
   widget=Button_BrowseToAsset
   slot {class=HorizontalBoxSlot, properties {verticalAlignment=VAlign_Center, padding=<ltrb 4 4 4 4>}}
  }
  Widget_CategoryData {
   widgetClassPath=/QuestUI/DebugTool/CommonWidgets/QuestDebugToolInfoField
   bIsVariable=true
   parent=HorizontalBox_2
   widget=Widget_CategoryData
   properties {header=Category Data}
   slot {class=HorizontalBoxSlot, properties {size {sizeRule=Fill, value=1}}}
  }
  Widget_DisplayName {
   widgetClassPath=&QuestDebugToolInfoField
   bIsVariable=true
   parent=HorizontalBox_2
   widget=Widget_DisplayName
   properties {header=Display Name}
   slot {class=HorizontalBoxSlot, properties {size {sizeRule=Fill, value=1}}}
  }
 ]
}
