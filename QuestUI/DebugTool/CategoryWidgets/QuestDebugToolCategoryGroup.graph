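# Quest debug tool widget that displays one category group: a category entry with its products beneath it.
# SetData takes a category data struct by reference and forwards it to the category entry, then to the products widget.
# A vertical box stacks the entry over a row where a faint, narrow spacer image sits left of the products, which fill the remaining width.
# The widget stores no variables of its own and has an empty event graph.
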WidgetBlueprint /QuestUI/DebugTool/CategoryWidgets/QuestDebugToolCategoryGroup
graphs [
 EventGraph {links [], nodes []}
 SetData {nodes [
   n1 = SetData!() -> (then, Data:"Quest Panel Debug Category Data Structure (by ref)") {is=title K2Node_FunctionEntry}
   > SetData!(
    execute
    self:Quest Debug Tool Category Entry Object Reference=Get Widget_Category() -> (Widget_Category:Quest Debug Tool Category Entry Object Reference)
    Data:"Quest Panel Debug Category Data Structure (by ref)"=@n1
   ) -> (then) {is=title K2Node_CallFunction}
   > SetData(
    self:Quest Debug Tool Category Products Object Reference=Get Widget_Products() -> (Widget_Products:Quest Debug Tool Category Products Object Reference)
    Data=@n1
   )
  ]}
]
variables []
widgets {
 info {inheritedWidgetCount=0, parentClass=&UserWidget, rootWidgetClass=&VerticalBox, widgetCount=5}
 widgets [
  VerticalBox_32 {widgetClassPath=&VerticalBox, widget=VerticalBox_32}
  Widget_Category {
   widgetClassPath=&/QuestDebugToolCategoryEntry
   bIsVariable=true
   parent=VerticalBox_32
   widget=Widget_Category
   slot {class=VerticalBoxSlot}
  }
  HorizontalBox_0 {widgetClassPath=&HorizontalBox, parent=VerticalBox_32, widget=HorizontalBox_0, slot {class=VerticalBoxSlot}}
  Image_Spacer_Vert_2 {
   widgetClassPath=&Image
   parent=HorizontalBox_0
   widget=Image_Spacer_Vert_2
   properties {
    brush {
     bIsDynamicallyLoaded=false
     drawAs=Image
     imageType=NoImage
     mirroring=NoMirror
     resourceName=None
     resourceObject=null
     tiling=NoTile
     imageSize=<4 32>
     margin=<ltrb 0 0 0 0>
     outlineSettings {
      bUseBrushTransparency=false
      roundingType=HalfHeightRadius
      width=0
      color {colorUseRule=UseColor_Specified, specifiedColor=<rgba 0 0 0 0>}
      cornerRadii=<0 0 0 0>
     }
     tintColor {colorUseRule=UseColor_Specified, specifiedColor=<rgba 1 1 1 1>}
     uVRegion {bIsValid=false, max=<0 0>, min=<0 0>}
    }
    colorAndOpacity=<rgba 1 1 1 0.1>
   }
   slot {class=HorizontalBoxSlot, properties {padding=<ltrb 8 8 8 8>}}
  }
  Widget_Products {
   widgetClassPath=&/QuestDebugToolCategoryProducts
   bIsVariable=true
   parent=HorizontalBox_0
   widget=Widget_Products
   slot {class=HorizontalBoxSlot, properties {size {sizeRule=Fill, value=1}}}
  }
 ]
}
